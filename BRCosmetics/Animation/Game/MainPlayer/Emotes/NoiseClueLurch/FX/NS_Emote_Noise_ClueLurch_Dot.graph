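# Niagara system for the NoiseClueLurch emote, drawing glowing dots in six concentric ground-facing rings.
# User.DiscAnim drives placement around rings, Outer/Middle/InnerCircle control brightness, and staggered loop-age color curves ripple orange and red across rings.
# Active Dot001 follows slopes through traces and scales by inverted rail float, while Dot002 scales by rail float plus front-end flags.
# The socket-sampling Dot emitter is disabled.
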NiagaraSystem /BRCosmetics/Animation/Game/MainPlayer/Emotes/NoiseClueLurch/FX/NS_Emote_Noise_ClueLurch_Dot
dependencies {
 usedDataInterfaces [
  &NiagaraDataInterfaceColorCurve
  &NiagaraDataInterfaceCurve
  &NiagaraDataInterfaceSpriteRendererInfo
  &NiagaraDataInterfaceCollisionQuery
  &NiagaraDataInterfaceDebugDraw
  &NiagaraDataInterfaceSocketReader
  &NiagaraDataInterfaceArrayFloat3
 ]
 usedDynamicInputs [
  &SelectVectorFromArray
  &SetBoolByFloatComparison
  &VectorLength
  /Niagara/DynamicInputs/Lerp/SmoothLerpOverTime_Float
  &Boolean_NotOperation
 ]
 usedModules [
  &SystemState
  &self:SystemSpawnScript.SystemScriptSource.SystemScriptGraph.NiagaraNodeAssignment_6.SetVariables_#1
  &self:SystemSpawnScript.SystemScriptSource.SystemScriptGraph.NiagaraNodeAssignment_1.SetVariables_#2
  &self:SystemSpawnScript.SystemScriptSource.SystemScriptGraph.NiagaraNodeAssignment_0.SetVariables_#3
  &BlendSpaceHelper
 ]
}
emitters {
 Dot {
  bEnabled=false
  emitterUpdateScript [
   &EmitterState(
    Life Cycle Mode=::NewEnumerator1
    Loop Delay=&RandomRangeFloat(Minimum=0.2, Maximum=0.5)
    Loop Duration=1
    UseLoopDelay=false
   )
   /Game/Effects/Niagara/Modules/Logic/UpdateValueStateChange(InputIntParam:NiagaraInt32=&MakeIntFromFloat(Float=$User.DiscAnim, Conversion Mode=::NewEnumerator1))
   &SpawnRate(SpawnRate=45, Use Spawn Probability=false)
   &SpawnRate(SpawnGroup=1, SpawnRate=40, Use Spawn Probability=false) {module=SpawnRate001}
   &SpawnRate(SpawnGroup=2, SpawnRate=30, Use Spawn Probability=false) {module=SpawnRate002}
   &SpawnRate(SpawnGroup=3, SpawnRate=20, Use Spawn Probability=false) {module=SpawnRate003}
   &SpawnRate(SpawnGroup=4, SpawnRate=12, Use Spawn Probability=false) {module=SpawnRate004}
   &SpawnRate(SpawnGroup=5, SpawnRate=4, Use Spawn Probability=false) {module=SpawnRate005}
  ]
  particleSpawnScript [
   &PartitionParticles(Number Of Partitions=5)
   &InitializeParticle(
    Lifetime=0.7
    Sprite Size Mode=::NewEnumerator3
    Uniform Sprite Size=&MergeHelper_FloatSwitch(0=7.5, 1=7.1, 2=6.7, 3=6.2, 4=5.8, 5=5.5, Index=&SpawnGroup(), Num=::NewEnumerator4)
    UsePositionOffset=false
    Write Material Random=true
   )
   &SocketLocation(
    Orientation Sampling=::NewEnumerator1
    Position Sampling=::NewEnumerator1
    SamplePosOffset=false
    Socket Reader=$System.Socket Reader
    Socket Sampling Mode=::NewEnumerator3
    VelocitySpeedLimit=false
   )
   &SpriteFacingAndAlignment(Facing Coordinate Space=::World, Sprite Facing=<0 0 1>, WriteAlignment=false, WriteFacing=true)
   &ShapeLocation(
    MaskSpawnGroup=false
    Offset=<0 0 5>
    Offset Coordinate Space=::World
    Offset Mode=::NewEnumerator0
    Override Seed=false
    Ring / Disc Distribution Mode:ENiagaraSphereDistributionMode=::NewEnumerator1
    Ring / Disc Mode:ENiagaraRingDiscMode=::NewEnumerator1
    Ring Radius=&MergeHelper_FloatSwitch(0=100, 1=80, 2=60, 3=40, 4=20, 5=6, Index=&SpawnGroup(), Num=::NewEnumerator4)
    Rotation Axis=<0 0 1>
    Rotation Coordinate Space=::World
    Rotation Mode=::NewEnumerator1
    Shape Primitive=::NewEnumerator4
    U Position=&Frac_Float(Float=&Multiply_Float(
      A=$Emitter.Age
      B=&Multiply_Float(
       A=$User.DiscAnim
       B=&FloatFromCurve(FloatCurve=DataInterface() {propertyValues {Curve {keys [<at 0 1>,<at 1 0>]}}})
      )
     ))
    UseOwnerScale=true
   ) {module=ShapeLocation002}
  ]
  particleUpdateScript [
   &ParticleState()
   &ScaleSpriteSize(Uniform Curve Sprite Scale=DataInterface() {propertyValues {Curve {
       defaultValue=1
       keys [
        {interpMode=RCIM_Cubic, tangentMode=RCTM_User, time=0, value=0}
        {arriveTangentWeight=0.094, interpMode=RCIM_Cubic, leaveTangentWeight=0.282, tangentMode=RCTM_User, time=0.1, value=0.9128}
        {arriveTangentWeight=0.3034, interpMode=RCIM_Cubic, leaveTangentWeight=0.1011, tangentMode=RCTM_User, time=0.9, value=0.7953}
        {tangentMode=RCTM_User, time=1, value=0}
       ]
      }}})
   &ScaleColor(
    Scale Alpha=&Multiply_Float(
     A=&FloatFromCurve(
      FloatCurve=DataInterface() {propertyValues {Curve {
         defaultValue=0
         keys [
          {interpMode=RCIM_Cubic, tangentMode=RCTM_User, time=0, value=0}
          {interpMode=RCIM_Cubic, tangentMode=RCTM_User, time=0.5, value=1}
          {tangentMode=RCTM_User, time=1, value=0}
         ]
        }}}
      Scale Curve=&MaskFloat(Mask Source=1, Mask Parameter=$Particles.PartitionParticles.Partition, Mask=0, Masked Float=0, Passthrough Float=1)
     )
     B=&MergeHelper_FloatSwitch(
      0=$User.OuterCircle
      1=$User.MiddleCircle
      2=$User.MiddleCircle
      3=$User.InnerCircle
      4=$User.InnerCircle
      5=$User.InnerCircle
      Index=&SpawnGroup()
      Num=::NewEnumerator4
     )
    )
    Scale RGB=&MakeVectorFromLinearColorRGB(Color=&Multiply_LinearColorByFloat(
      Linear Color=&MergeHelper_ColorSwitch(
       0=&ColorFromCurve(
        ColorCurve=DataInterface() {propertyValues {
          AlphaCurve {keys [<at 0 1 RCIM_Constant>,<at 1 0 RCIM_Cubic>]}
          BlueCurve {keys [<at 0 0 RCIM_Constant>,<at 0.0628 0 RCIM_Constant>,<at 0.1402 0 RCIM_Constant>,<at 0.2131 0 RCIM_Cubic>,<at 0.2738 0 RCIM_Cubic>,<at 0.3967 0 RCIM_Cubic>,<at 0.5303 0 RCIM_Cubic>,<at 0.6016 0 RCIM_Constant>,<at 0.7033 0 RCIM_Cubic>,<at 0.8718 0 RCIM_Cubic>,<at 0.9902 0 RCIM_Cubic>]}
          GreenCurve {keys [
            <at 0 0.638 RCIM_Constant>
            <at 0.0628 0.0331 RCIM_Constant>
            <at 0.1402 0.6376 RCIM_Constant>
            {arriveTangent=-4.5259, interpMode=RCIM_Cubic, leaveTangent=-4.5259, time=0.2131, value=0.6376}
            {arriveTangent=-3.2915, interpMode=RCIM_Cubic, leaveTangent=-3.2915, time=0.2738, value=0.0331}
            {arriveTangent=2.3567, interpMode=RCIM_Cubic, leaveTangent=2.3567, time=0.3967, value=0.0331}
            {arriveTangent=2.9502, interpMode=RCIM_Cubic, leaveTangent=2.9502, time=0.5303, value=0.6376}
            <at 0.6016 0.6376 RCIM_Constant>
            {arriveTangent=-2.2375, interpMode=RCIM_Cubic, leaveTangent=-2.2375, time=0.7033, value=0.0331}
            {arriveTangent=2.1073, interpMode=RCIM_Cubic, leaveTangent=2.1073, time=0.8718, value=0.0331}
            <at 0.9902 0.6376 RCIM_Cubic>
           ]}
          RedCurve {keys [<at 0 1 RCIM_Constant>,<at 0.0628 1 RCIM_Constant>,<at 0.1402 1 RCIM_Constant>,<at 0.2131 1 RCIM_Cubic>,<at 0.2738 1 RCIM_Cubic>,<at 0.3967 1 RCIM_Cubic>,<at 0.5303 1 RCIM_Cubic>,<at 0.6016 1 RCIM_Constant>,<at 0.7033 1 RCIM_Cubic>,<at 0.8718 1 RCIM_Cubic>,<at 0.9902 1 RCIM_Cubic>]}
         }}
        CurveIndex=$System.NormalizedLoopAge
       )
       1=&ColorFromCurve(
        ColorCurve=DataInterface() {propertyValues {
          AlphaCurve {keys [<at 0 1 RCIM_Constant>,<at 1 0 RCIM_Cubic>]}
          BlueCurve {keys [<at 0.0492 0 RCIM_Constant>,<at 0.1114 0 RCIM_Constant>,<at 0.2101 0>,<at 0.2829 0>,<at 0.3694 0>,<at 0.4848 0>,<at 0.5743 0 RCIM_Cubic>,<at 0.6578 0 RCIM_Constant>,<at 0.7337 0>,<at 0.9462 0>,<at 0.9902 0>]}
          GreenCurve {keys [
            <at 0.0492 0.638 RCIM_Constant>
            <at 0.1114 0.0331 RCIM_Constant>
            <at 0.2101 0.6376>
            <at 0.2829 0.6376>
            <at 0.3694 0.0331>
            <at 0.4848 0.0331>
            {arriveTangent=3.4936, interpMode=RCIM_Cubic, leaveTangent=3.4936, time=0.5743, value=0.6376}
            <at 0.6578 0.6376 RCIM_Constant>
            <at 0.7337 0.0331>
            <at 0.9462 0.0331>
            <at 0.9902 0.6376>
           ]}
          RedCurve {keys [<at 0.0492 1 RCIM_Constant>,<at 0.1114 1 RCIM_Constant>,<at 0.2101 1>,<at 0.2829 1>,<at 0.3694 1>,<at 0.4848 1>,<at 0.5743 1 RCIM_Cubic>,<at 0.6578 1 RCIM_Constant>,<at 0.7337 1>,<at 0.9462 1>,<at 0.9902 1>]}
         }}
        CurveIndex=$System.NormalizedLoopAge
       )
       2=&ColorFromCurve(
        ColorCurve=DataInterface() {propertyValues {
          AlphaCurve {keys [<at 0 1 RCIM_Constant>,<at 1 0 RCIM_Cubic>]}
          BlueCurve {keys [<at 0.0932 0 RCIM_Constant>,<at 0.1569 0 RCIM_Constant>,<at 0.2662 0>,<at 0.2996 0>,<at 0.4286 0>,<at 0.4848 0>,<at 0.6108 0>,<at 0.6578 0 RCIM_Constant>,<at 0.805 0 RCIM_Constant>,<at 0.9462 0>,<at 0.9902 0>]}
          GreenCurve {keys [<at 0.0932 0.638 RCIM_Constant>,<at 0.1569 0.0331 RCIM_Constant>,<at 0.2662 0.6376>,<at 0.2996 0.6376>,<at 0.4286 0.0331>,<at 0.4848 0.0331>,<at 0.6108 0.6376>,<at 0.6578 0.6376 RCIM_Constant>,<at 0.805 0.0331 RCIM_Constant>,<at 0.9462 0.0331>,<at 0.9902 0.6376>]}
          RedCurve {keys [<at 0.0932 1 RCIM_Constant>,<at 0.1569 1 RCIM_Constant>,<at 0.2662 1>,<at 0.2996 1>,<at 0.4286 1>,<at 0.4848 1>,<at 0.6108 1>,<at 0.6578 1 RCIM_Constant>,<at 0.805 1 RCIM_Constant>,<at 0.9462 1>,<at 0.9902 1>]}
         }}
        CurveIndex=$System.NormalizedLoopAge
       )
       3=&ColorFromCurve(
        ColorCurve=DataInterface() {propertyValues {
          AlphaCurve {keys [<at 0 1 RCIM_Constant>,<at 1 0 RCIM_Cubic>]}
          BlueCurve {keys [<at 0.1357 0 RCIM_Constant>,<at 0.207 0 RCIM_Constant>,<at 0.3011 0>,<at 0.3527 0>,<at 0.4969 0>,<at 0.5485 0>,<at 0.6699 0>,<at 0.7853 0 RCIM_Constant>,<at 0.8551 0 RCIM_Constant>,<at 0.9583 0>,<at 0.9902 0>]}
          GreenCurve {keys [<at 0.1357 0.638 RCIM_Constant>,<at 0.207 0.0331 RCIM_Constant>,<at 0.3011 0.6376>,<at 0.3527 0.6376>,<at 0.4969 0.0331>,<at 0.5485 0.0331>,<at 0.6699 0.6376>,<at 0.7853 0.6376 RCIM_Constant>,<at 0.8551 0.0331 RCIM_Constant>,<at 0.9583 0.0331>,<at 0.9902 0.6376>]}
          RedCurve {keys [<at 0.1357 1 RCIM_Constant>,<at 0.207 1 RCIM_Constant>,<at 0.3011 1>,<at 0.3527 1>,<at 0.4969 1>,<at 0.5485 1>,<at 0.6699 1>,<at 0.7853 1 RCIM_Constant>,<at 0.8551 1 RCIM_Constant>,<at 0.9583 1>,<at 0.9902 1>]}
         }}
        CurveIndex=$System.NormalizedLoopAge
       )
       4=&ColorFromCurve(
        ColorCurve=DataInterface() {propertyValues {
          AlphaCurve {keys [<at 0 1 RCIM_Cubic>,<at 1 0 RCIM_Cubic>]}
          BlueCurve {keys [<at 0.1767 0 RCIM_Constant>,<at 0.2495 0 RCIM_Constant>,<at 0.3527 0 RCIM_Constant>,<at 0.4134 0>,<at 0.5485 0>,<at 0.6153 0>,<at 0.7352 0>,<at 0.7853 0 RCIM_Constant>,<at 0.9022 0 RCIM_Constant>,<at 0.9583 0>,<at 1 0>]}
          GreenCurve {keys [<at 0.1767 0.638 RCIM_Constant>,<at 0.2495 0.0331 RCIM_Constant>,<at 0.3527 0.6376 RCIM_Constant>,<at 0.4134 0.6376>,<at 0.5485 0.0331>,<at 0.6153 0.0331>,<at 0.7352 0.6376>,<at 0.7853 0.6376 RCIM_Constant>,<at 0.9022 0.0331 RCIM_Constant>,<at 0.9583 0.0331>,<at 1 0.6376>]}
          RedCurve {keys [<at 0.1767 1 RCIM_Constant>,<at 0.2495 1 RCIM_Constant>,<at 0.3527 1 RCIM_Constant>,<at 0.4134 1>,<at 0.5485 1>,<at 0.6153 1>,<at 0.7352 1>,<at 0.7853 1 RCIM_Constant>,<at 0.9022 1 RCIM_Constant>,<at 0.9583 1>,<at 1 1>]}
         }}
        CurveIndex=$System.NormalizedLoopAge
       )
       5=&ColorFromCurve(
        ColorCurve=DataInterface() {propertyValues {
          AlphaCurve {keys [<at 0 1 RCIM_Cubic>,<at 1 0 RCIM_Cubic>]}
          BlueCurve {keys [<at 0.2161 0 RCIM_Constant>,<at 0.3011 0 RCIM_Constant>,<at 0.3755 0 RCIM_Constant>,<at 0.4772 0>,<at 0.6199 0>,<at 0.6791 0>,<at 0.7853 0 RCIM_Constant>,<at 0.8384 0 RCIM_Constant>,<at 0.934 0 RCIM_Constant>,<at 0.972 0>,<at 1 0>]}
          GreenCurve {keys [<at 0.2161 0.638 RCIM_Constant>,<at 0.3011 0.0331 RCIM_Constant>,<at 0.3755 0.6376 RCIM_Constant>,<at 0.4772 0.6376>,<at 0.6199 0.0331>,<at 0.6791 0.0331>,<at 0.7853 0.6376 RCIM_Constant>,<at 0.8384 0.6376 RCIM_Constant>,<at 0.934 0.0331 RCIM_Constant>,<at 0.972 0.0331>,<at 1 0.6376>]}
          RedCurve {keys [<at 0.2161 1 RCIM_Constant>,<at 0.3011 1 RCIM_Constant>,<at 0.3755 1 RCIM_Constant>,<at 0.4772 1>,<at 0.6199 1>,<at 0.6791 1>,<at 0.7853 1 RCIM_Constant>,<at 0.8384 1 RCIM_Constant>,<at 0.934 1 RCIM_Constant>,<at 0.972 1>,<at 1 1>]}
         }}
        CurveIndex=$System.NormalizedLoopAge
       )
       Index=&SpawnGroup()
       Num=::NewEnumerator4
      )
      Float=10
     ))
    ScaleA=true
    ScaleRGB=true
   )
   &SolveForcesAndVelocity(Clamp Velocity=false, Limit Acceleration=false)
   &SubUVAnimation(
    Evaluation Type=::NewEnumerator0
    Output Alternate Renderer Binding=true
    "Sprite Renderer "=DataInterface() {propertyValues {SpriteRenderer=&self:Dot_1.NiagaraSpriteRendererProperties_2}}
    SubUV Animation Mode:ENiagara_SubUVLookupModeV2=::NewEnumerator2
    UseEndFrame=false
    UseStartFrame=false
   )
   /SetVariables(
    Particles.GlowColor=&Multiply_LinearColorByFloat(
     Linear Color=$Particles.Color
     Float=&Multiply_Float(
      A=&FloatFromCurve(
       FloatCurve=DataInterface() {propertyValues {Curve {
          defaultValue=1
          keys [{interpMode=RCIM_Cubic, tangentMode=RCTM_User, time=0, value=1}, {interpMode=RCIM_Cubic, tangentMode=RCTM_User, time=1, value=0}]
         }}}
       Scale Curve=&RandomRangeFloat(Minimum=0.5, Maximum=0.7)
      )
      B=&MergeHelper_FloatSwitch(
       0=$User.OuterCircle
       1=$User.MiddleCircle
       2=$User.MiddleCircle
       3=$User.InnerCircle
       4=$User.InnerCircle
       5=$User.InnerCircle
       Index=&SpawnGroup()
       Num=::NewEnumerator4
      )
     )
    )
    Particles.GlowSize=&Multiply_Vector2D_ByFloat(Float=&RandomRangeFloat(Minimum=3, Maximum=1.7), Vector2D=$Particles.SpriteSize)
   ) {isSetParameters=true, module=SetVariables_#4}
   /SetVariables(
    Particles.RaysColor:LinearColor=&Multiply_LinearColorByFloat(
     Linear Color=$Particles.Color
     Float=&Multiply_Float(
      A=&FloatFromCurve(
       FloatCurve=DataInterface() {propertyValues {Curve {
          defaultValue=1
          keys [{interpMode=RCIM_Cubic, tangentMode=RCTM_User, time=0, value=1}, {interpMode=RCIM_Cubic, tangentMode=RCTM_User, time=1, value=0}]
         }}}
       Scale Curve=&RandomRangeFloat(Minimum=0.6, Maximum=1.2)
      )
      B=&MergeHelper_FloatSwitch(
       0=$User.OuterCircle
       1=$User.MiddleCircle
       2=$User.MiddleCircle
       3=$User.InnerCircle
       4=$User.InnerCircle
       5=$User.InnerCircle
       Index=&SpawnGroup()
       Num=::NewEnumerator4
      )
     )
    )
    Particles.RaysSize:Vector2f=&Multiply_Vector2D(A=&RandomRangeVector2D(Minimum=<1.8 2.5>, Maximum=<2.8 5>), B=$Particles.SpriteSize)
   ) {isSetParameters=true, module=SetVariables_#5}
  ]
  rendererClasses [&NiagaraSpriteRendererProperties]
  renderers [
   {rendererClass=&NiagaraSpriteRendererProperties}
   {rendererClass=&NiagaraSpriteRendererProperties, rendererIndex=1}
   {rendererClass=&NiagaraSpriteRendererProperties, rendererIndex=2}
  ]
 }
 Dot001 {
  emitterUpdateScript [
   &EmitterState(
    Life Cycle Mode=::NewEnumerator1
    Loop Delay=&RandomRangeFloat(Minimum=0.2, Maximum=0.5)
    Loop Duration=1
    UseLoopDelay=false
   )
   &UpdateValueStateChange(InputIntParam:NiagaraInt32=&MakeIntFromFloat(Float=$User.DiscAnim, Conversion Mode=::NewEnumerator1))
   &SpawnRate(SpawnRate=45, Use Spawn Probability=false)
   &SpawnRate(SpawnGroup=1, SpawnRate=40, Use Spawn Probability=false) {module=SpawnRate001}
   &SpawnRate(SpawnGroup=2, SpawnRate=30, Use Spawn Probability=false) {module=SpawnRate002}
   &SpawnRate(SpawnGroup=3, SpawnRate=20, Use Spawn Probability=false) {module=SpawnRate003}
   &SpawnRate(SpawnGroup=4, SpawnRate=12, Use Spawn Probability=false) {module=SpawnRate004}
   &SpawnRate(SpawnGroup=5, SpawnRate=4, Use Spawn Probability=false) {module=SpawnRate005}
  ]
  particleSpawnScript [
   &PartitionParticles(Number Of Partitions=5)
   &InitializeParticle(
    Lifetime=0.7
    Sprite Size Mode=::NewEnumerator3
    Uniform Sprite Size=&MergeHelper_FloatSwitch(0=7.5, 1=7.1, 2=6.7, 3=6.2, 4=5.8, 5=5.5, Index=&SpawnGroup(), Num=::NewEnumerator4)
    UsePositionOffset=false
    Write Material Random=true
   )
   &SpriteFacingAndAlignment(Facing Coordinate Space=::World, Sprite Facing=<0 0 1>, WriteAlignment=false, WriteFacing=true)
   &ShapeLocation(
    MaskSpawnGroup=false
    Offset=<0 0 5>
    Offset Coordinate Space=::World
    Offset Mode=::NewEnumerator0
    Override Seed=false
    Ring / Disc Distribution Mode:ENiagaraSphereDistributionMode=::NewEnumerator1
    Ring / Disc Mode:ENiagaraRingDiscMode=::NewEnumerator1
    Ring Radius=&MergeHelper_FloatSwitch(0=100, 1=80, 2=60, 3=40, 4=20, 5=6, Index=&SpawnGroup(), Num=::NewEnumerator4)
    Rotation Axis=<0 0 1>
    Rotation Coordinate Space=::World
    Rotation Mode=::NewEnumerator1
    Shape Primitive=::NewEnumerator4
    U Position=&Frac_Float(Float=&Multiply_Float(
      A=$Emitter.Age
      B=&Multiply_Float(
       A=$User.DiscAnim
       B=&FloatFromCurve(FloatCurve=DataInterface() {propertyValues {Curve {keys [<at 0 1>,<at 1 0>]}}})
      )
     ))
    UseOwnerScale=true
   ) {module=ShapeLocation002}
  ]
  particleUpdateScript [
   &ParticleState()
   &ScaleSpriteSize(Uniform Curve Sprite Scale=DataInterface() {propertyValues {Curve {
       defaultValue=1
       keys [
        {interpMode=RCIM_Cubic, tangentMode=RCTM_User, time=0, value=0}
        {arriveTangentWeight=0.094, interpMode=RCIM_Cubic, leaveTangentWeight=0.282, tangentMode=RCTM_User, time=0.1, value=0.9128}
        {arriveTangentWeight=0.3034, interpMode=RCIM_Cubic, leaveTangentWeight=0.1011, tangentMode=RCTM_User, time=0.9, value=0.7953}
        {tangentMode=RCTM_User, time=1, value=0}
       ]
      }}})
   &ScaleSpriteSize(Scale Sprite Size Mode=::NewEnumerator0, Uniform Scale Factor=$System.Rail Float Inverted) {module=ScaleSpriteSize001}
   &ScaleColor(
    Scale Alpha=&Multiply_Float(
     A=&FloatFromCurve(
      FloatCurve=DataInterface() {propertyValues {Curve {
         defaultValue=0
         keys [
          {interpMode=RCIM_Cubic, tangentMode=RCTM_User, time=0, value=0}
          {interpMode=RCIM_Cubic, tangentMode=RCTM_User, time=0.5, value=1}
          {tangentMode=RCTM_User, time=1, value=0}
         ]
        }}}
      Scale Curve=&MaskFloat(Mask Source=1, Mask Parameter=$Particles.PartitionParticles.Partition, Mask=0, Masked Float=0, Passthrough Float=1)
     )
     B=&MergeHelper_FloatSwitch(
      0=$User.OuterCircle
      1=$User.MiddleCircle
      2=$User.MiddleCircle
      3=$User.InnerCircle
      4=$User.InnerCircle
      5=$User.InnerCircle
      Index=&SpawnGroup()
      Num=::NewEnumerator4
     )
    )
    Scale RGB=&MakeVectorFromLinearColorRGB(Color=&Multiply_LinearColorByFloat(
      Linear Color=&MergeHelper_ColorSwitch(
       0=&ColorFromCurve(
        ColorCurve=DataInterface() {propertyValues {
          AlphaCurve {keys [<at 0 1 RCIM_Constant>,<at 1 0 RCIM_Cubic>]}
          BlueCurve {keys [<at 0 0 RCIM_Constant>,<at 0.0628 0 RCIM_Constant>,<at 0.1402 0 RCIM_Constant>,<at 0.2131 0 RCIM_Cubic>,<at 0.2738 0 RCIM_Cubic>,<at 0.3967 0 RCIM_Cubic>,<at 0.5303 0 RCIM_Cubic>,<at 0.6016 0 RCIM_Constant>,<at 0.7033 0 RCIM_Cubic>,<at 0.8718 0 RCIM_Cubic>,<at 0.9902 0 RCIM_Cubic>]}
          GreenCurve {keys [
            <at 0 0.638 RCIM_Constant>
            <at 0.0628 0.0331 RCIM_Constant>
            <at 0.1402 0.6376 RCIM_Constant>
            {arriveTangent=-4.5259, interpMode=RCIM_Cubic, leaveTangent=-4.5259, time=0.2131, value=0.6376}
            {arriveTangent=-3.2915, interpMode=RCIM_Cubic, leaveTangent=-3.2915, time=0.2738, value=0.0331}
            {arriveTangent=2.3567, interpMode=RCIM_Cubic, leaveTangent=2.3567, time=0.3967, value=0.0331}
            {arriveTangent=2.9502, interpMode=RCIM_Cubic, leaveTangent=2.9502, time=0.5303, value=0.6376}
            <at 0.6016 0.6376 RCIM_Constant>
            {arriveTangent=-2.2375, interpMode=RCIM_Cubic, leaveTangent=-2.2375, time=0.7033, value=0.0331}
            {arriveTangent=2.1073, interpMode=RCIM_Cubic, leaveTangent=2.1073, time=0.8718, value=0.0331}
            <at 0.9902 0.6376 RCIM_Cubic>
           ]}
          RedCurve {keys [<at 0 1 RCIM_Constant>,<at 0.0628 1 RCIM_Constant>,<at 0.1402 1 RCIM_Constant>,<at 0.2131 1 RCIM_Cubic>,<at 0.2738 1 RCIM_Cubic>,<at 0.3967 1 RCIM_Cubic>,<at 0.5303 1 RCIM_Cubic>,<at 0.6016 1 RCIM_Constant>,<at 0.7033 1 RCIM_Cubic>,<at 0.8718 1 RCIM_Cubic>,<at 0.9902 1 RCIM_Cubic>]}
         }}
        CurveIndex=$System.NormalizedLoopAge
       )
       1=&ColorFromCurve(
        ColorCurve=DataInterface() {propertyValues {
          AlphaCurve {keys [<at 0 1 RCIM_Constant>,<at 1 0 RCIM_Cubic>]}
          BlueCurve {keys [<at 0.0492 0 RCIM_Constant>,<at 0.1114 0 RCIM_Constant>,<at 0.2101 0>,<at 0.2829 0>,<at 0.3694 0>,<at 0.4848 0>,<at 0.5743 0 RCIM_Cubic>,<at 0.6578 0 RCIM_Constant>,<at 0.7337 0>,<at 0.9462 0>,<at 0.9902 0>]}
          GreenCurve {keys [
            <at 0.0492 0.638 RCIM_Constant>
            <at 0.1114 0.0331 RCIM_Constant>
            <at 0.2101 0.6376>
            <at 0.2829 0.6376>
            <at 0.3694 0.0331>
            <at 0.4848 0.0331>
            {arriveTangent=3.4936, interpMode=RCIM_Cubic, leaveTangent=3.4936, time=0.5743, value=0.6376}
            <at 0.6578 0.6376 RCIM_Constant>
            <at 0.7337 0.0331>
            <at 0.9462 0.0331>
            <at 0.9902 0.6376>
           ]}
          RedCurve {keys [<at 0.0492 1 RCIM_Constant>,<at 0.1114 1 RCIM_Constant>,<at 0.2101 1>,<at 0.2829 1>,<at 0.3694 1>,<at 0.4848 1>,<at 0.5743 1 RCIM_Cubic>,<at 0.6578 1 RCIM_Constant>,<at 0.7337 1>,<at 0.9462 1>,<at 0.9902 1>]}
         }}
        CurveIndex=$System.NormalizedLoopAge
       )
       2=&ColorFromCurve(
        ColorCurve=DataInterface() {propertyValues {
          AlphaCurve {keys [<at 0 1 RCIM_Constant>,<at 1 0 RCIM_Cubic>]}
          BlueCurve {keys [<at 0.0932 0 RCIM_Constant>,<at 0.1569 0 RCIM_Constant>,<at 0.2662 0>,<at 0.2996 0>,<at 0.4286 0>,<at 0.4848 0>,<at 0.6108 0>,<at 0.6578 0 RCIM_Constant>,<at 0.805 0 RCIM_Constant>,<at 0.9462 0>,<at 0.9902 0>]}
          GreenCurve {keys [<at 0.0932 0.638 RCIM_Constant>,<at 0.1569 0.0331 RCIM_Constant>,<at 0.2662 0.6376>,<at 0.2996 0.6376>,<at 0.4286 0.0331>,<at 0.4848 0.0331>,<at 0.6108 0.6376>,<at 0.6578 0.6376 RCIM_Constant>,<at 0.805 0.0331 RCIM_Constant>,<at 0.9462 0.0331>,<at 0.9902 0.6376>]}
          RedCurve {keys [<at 0.0932 1 RCIM_Constant>,<at 0.1569 1 RCIM_Constant>,<at 0.2662 1>,<at 0.2996 1>,<at 0.4286 1>,<at 0.4848 1>,<at 0.6108 1>,<at 0.6578 1 RCIM_Constant>,<at 0.805 1 RCIM_Constant>,<at 0.9462 1>,<at 0.9902 1>]}
         }}
        CurveIndex=$System.NormalizedLoopAge
       )
       3=&ColorFromCurve(
        ColorCurve=DataInterface() {propertyValues {
          AlphaCurve {keys [<at 0 1 RCIM_Constant>,<at 1 0 RCIM_Cubic>]}
          BlueCurve {keys [<at 0.1357 0 RCIM_Constant>,<at 0.207 0 RCIM_Constant>,<at 0.3011 0>,<at 0.3527 0>,<at 0.4969 0>,<at 0.5485 0>,<at 0.6699 0>,<at 0.7853 0 RCIM_Constant>,<at 0.8551 0 RCIM_Constant>,<at 0.9583 0>,<at 0.9902 0>]}
          GreenCurve {keys [<at 0.1357 0.638 RCIM_Constant>,<at 0.207 0.0331 RCIM_Constant>,<at 0.3011 0.6376>,<at 0.3527 0.6376>,<at 0.4969 0.0331>,<at 0.5485 0.0331>,<at 0.6699 0.6376>,<at 0.7853 0.6376 RCIM_Constant>,<at 0.8551 0.0331 RCIM_Constant>,<at 0.9583 0.0331>,<at 0.9902 0.6376>]}
          RedCurve {keys [<at 0.1357 1 RCIM_Constant>,<at 0.207 1 RCIM_Constant>,<at 0.3011 1>,<at 0.3527 1>,<at 0.4969 1>,<at 0.5485 1>,<at 0.6699 1>,<at 0.7853 1 RCIM_Constant>,<at 0.8551 1 RCIM_Constant>,<at 0.9583 1>,<at 0.9902 1>]}
         }}
        CurveIndex=$System.NormalizedLoopAge
       )
       4=&ColorFromCurve(
        ColorCurve=DataInterface() {propertyValues {
          AlphaCurve {keys [<at 0 1 RCIM_Cubic>,<at 1 0 RCIM_Cubic>]}
          BlueCurve {keys [<at 0.1767 0 RCIM_Constant>,<at 0.2495 0 RCIM_Constant>,<at 0.3527 0 RCIM_Constant>,<at 0.4134 0>,<at 0.5485 0>,<at 0.6153 0>,<at 0.7352 0>,<at 0.7853 0 RCIM_Constant>,<at 0.9022 0 RCIM_Constant>,<at 0.9583 0>,<at 1 0>]}
          GreenCurve {keys [<at 0.1767 0.638 RCIM_Constant>,<at 0.2495 0.0331 RCIM_Constant>,<at 0.3527 0.6376 RCIM_Constant>,<at 0.4134 0.6376>,<at 0.5485 0.0331>,<at 0.6153 0.0331>,<at 0.7352 0.6376>,<at 0.7853 0.6376 RCIM_Constant>,<at 0.9022 0.0331 RCIM_Constant>,<at 0.9583 0.0331>,<at 1 0.6376>]}
          RedCurve {keys [<at 0.1767 1 RCIM_Constant>,<at 0.2495 1 RCIM_Constant>,<at 0.3527 1 RCIM_Constant>,<at 0.4134 1>,<at 0.5485 1>,<at 0.6153 1>,<at 0.7352 1>,<at 0.7853 1 RCIM_Constant>,<at 0.9022 1 RCIM_Constant>,<at 0.9583 1>,<at 1 1>]}
         }}
        CurveIndex=$System.NormalizedLoopAge
       )
       5=&ColorFromCurve(
        ColorCurve=DataInterface() {propertyValues {
          AlphaCurve {keys [<at 0 1 RCIM_Cubic>,<at 1 0 RCIM_Cubic>]}
          BlueCurve {keys [<at 0.2161 0 RCIM_Constant>,<at 0.3011 0 RCIM_Constant>,<at 0.3755 0 RCIM_Constant>,<at 0.4772 0>,<at 0.6199 0>,<at 0.6791 0>,<at 0.7853 0 RCIM_Constant>,<at 0.8384 0 RCIM_Constant>,<at 0.934 0 RCIM_Constant>,<at 0.972 0>,<at 1 0>]}
          GreenCurve {keys [<at 0.2161 0.638 RCIM_Constant>,<at 0.3011 0.0331 RCIM_Constant>,<at 0.3755 0.6376 RCIM_Constant>,<at 0.4772 0.6376>,<at 0.6199 0.0331>,<at 0.6791 0.0331>,<at 0.7853 0.6376 RCIM_Constant>,<at 0.8384 0.6376 RCIM_Constant>,<at 0.934 0.0331 RCIM_Constant>,<at 0.972 0.0331>,<at 1 0.6376>]}
          RedCurve {keys [<at 0.2161 1 RCIM_Constant>,<at 0.3011 1 RCIM_Constant>,<at 0.3755 1 RCIM_Constant>,<at 0.4772 1>,<at 0.6199 1>,<at 0.6791 1>,<at 0.7853 1 RCIM_Constant>,<at 0.8384 1 RCIM_Constant>,<at 0.934 1 RCIM_Constant>,<at 0.972 1>,<at 1 1>]}
         }}
        CurveIndex=$System.NormalizedLoopAge
       )
       Index=&SpawnGroup()
       Num=::NewEnumerator4
      )
      Float=10
     ))
    ScaleA=true
    ScaleRGB=true
   )
   &SolveForcesAndVelocity(Clamp Velocity=false, Limit Acceleration=false)
   &SubUVAnimation(
    Evaluation Type=::NewEnumerator0
    Output Alternate Renderer Binding=true
    "Sprite Renderer "=DataInterface() {propertyValues {SpriteRenderer=&self:Dot001_0.NiagaraSpriteRendererProperties_2}}
    SubUV Animation Mode:ENiagara_SubUVLookupModeV2=::NewEnumerator2
    UseEndFrame=false
    UseStartFrame=false
   )
   /SetVariables(
    Particles.GlowColor=&Multiply_LinearColorByFloat(
     Linear Color=$Particles.Color
     Float=&Multiply_Float(
      A=&FloatFromCurve(
       FloatCurve=DataInterface() {propertyValues {Curve {
          defaultValue=1
          keys [{interpMode=RCIM_Cubic, tangentMode=RCTM_User, time=0, value=1}, {interpMode=RCIM_Cubic, tangentMode=RCTM_User, time=1, value=0}]
         }}}
       Scale Curve=&RandomRangeFloat(Minimum=0.5, Maximum=0.7)
      )
      B=&MergeHelper_FloatSwitch(
       0=$User.OuterCircle
       1=$User.MiddleCircle
       2=$User.MiddleCircle
       3=$User.InnerCircle
       4=$User.InnerCircle
       5=$User.InnerCircle
       Index=&SpawnGroup()
       Num=::NewEnumerator4
      )
     )
    )
    Particles.GlowSize=&Multiply_Vector2D_ByFloat(Float=&RandomRangeFloat(Minimum=3, Maximum=1.7), Vector2D=$Particles.SpriteSize)
   ) {isSetParameters=true, module=SetVariables_#4}
   /SetVariables(
    Particles.RaysColor:LinearColor=&Multiply_LinearColorByFloat(
     Linear Color=$Particles.Color
     Float=&Multiply_Float(
      A=&FloatFromCurve(
       FloatCurve=DataInterface() {propertyValues {Curve {
          defaultValue=1
          keys [{interpMode=RCIM_Cubic, tangentMode=RCTM_User, time=0, value=1}, {interpMode=RCIM_Cubic, tangentMode=RCTM_User, time=1, value=0}]
         }}}
       Scale Curve=&RandomRangeFloat(Minimum=0.6, Maximum=1.2)
      )
      B=&MergeHelper_FloatSwitch(
       0=$User.OuterCircle
       1=$User.MiddleCircle
       2=$User.MiddleCircle
       3=$User.InnerCircle
       4=$User.InnerCircle
       5=$User.InnerCircle
       Index=&SpawnGroup()
       Num=::NewEnumerator4
      )
     )
    )
    Particles.RaysSize:Vector2f=&Multiply_Vector2D(A=&RandomRangeVector2D(Minimum=<1.8 2.5>, Maximum=<2.8 5>), B=$Particles.SpriteSize)
   ) {isSetParameters=true, module=SetVariables_#5}
   &self:Slope_Position_02(
    Apply Position=true
    Apply Rotation=true
    Fallback Position:NiagaraPosition=&AddVectorToPosition(Position=$Particles.Initial.Position, Vector=<0 0 0>)
    Height Adjustment/Offset=2
    Slope Position:NiagaraPosition=$Particles.Initial.ShapeLocation002.ShapePosition
    Trace Length=120
    Trace Origin Height=85
    bIsFrontEnd:NiagaraBool=$User.bIsFrontEndPreview
   ) {module=Slope_Position_02}
   &BlendSpace(Translation Contribution=0) {module=BlendSpace001}
  ]
  rendererClasses [&NiagaraSpriteRendererProperties]
  renderers [
   {rendererClass=&NiagaraSpriteRendererProperties}
   {rendererClass=&NiagaraSpriteRendererProperties, rendererIndex=1}
   {rendererClass=&NiagaraSpriteRendererProperties, rendererIndex=2}
  ]
 }
 Dot002 {
  emitterUpdateScript [
   &EmitterState(
    Life Cycle Mode=::NewEnumerator1
    Loop Delay=&RandomRangeFloat(Minimum=0.2, Maximum=0.5)
    Loop Duration=1
    UseLoopDelay=false
   )
   &UpdateValueStateChange(InputIntParam:NiagaraInt32=&MakeIntFromFloat(Float=$User.DiscAnim, Conversion Mode=::NewEnumerator1))
   &SpawnRate(SpawnRate=45, Use Spawn Probability=false)
   &SpawnRate(SpawnGroup=1, SpawnRate=40, Use Spawn Probability=false) {module=SpawnRate001}
   &SpawnRate(SpawnGroup=2, SpawnRate=30, Use Spawn Probability=false) {module=SpawnRate002}
   &SpawnRate(SpawnGroup=3, SpawnRate=20, Use Spawn Probability=false) {module=SpawnRate003}
   &SpawnRate(SpawnGroup=4, SpawnRate=12, Use Spawn Probability=false) {module=SpawnRate004}
   &SpawnRate(SpawnGroup=5, SpawnRate=4, Use Spawn Probability=false) {module=SpawnRate005}
  ]
  particleSpawnScript [
   &PartitionParticles(Number Of Partitions=5)
   &InitializeParticle(
    Lifetime=0.7
    Sprite Size Mode=::NewEnumerator3
    Uniform Sprite Size=&MergeHelper_FloatSwitch(0=7.5, 1=7.1, 2=6.7, 3=6.2, 4=5.8, 5=5.5, Index=&SpawnGroup(), Num=::NewEnumerator4)
    UsePositionOffset=false
    Write Material Random=true
   )
   &SpriteFacingAndAlignment(Facing Coordinate Space=::World, Sprite Facing=<0 0 1>, WriteAlignment=false, WriteFacing=true)
   &ShapeLocation(
    MaskSpawnGroup=false
    Offset=<0 0 5>
    Offset Coordinate Space=::World
    Offset Mode=::NewEnumerator0
    Override Seed=false
    Ring / Disc Distribution Mode:ENiagaraSphereDistributionMode=::NewEnumerator1
    Ring / Disc Mode:ENiagaraRingDiscMode=::NewEnumerator1
    Ring Radius=&MergeHelper_FloatSwitch(0=100, 1=80, 2=60, 3=40, 4=20, 5=6, Index=&SpawnGroup(), Num=::NewEnumerator4)
    Rotation Axis=<0 0 1>
    Rotation Coordinate Space=::World
    Rotation Mode=::NewEnumerator1
    Shape Primitive=::NewEnumerator4
    U Position=&Frac_Float(Float=&Multiply_Float(
      A=$Emitter.Age
      B=&Multiply_Float(
       A=$User.DiscAnim
       B=&FloatFromCurve(FloatCurve=DataInterface() {propertyValues {Curve {keys [<at 0 1>,<at 1 0>]}}})
      )
     ))
    UseOwnerScale=true
   ) {module=ShapeLocation002}
  ]
  particleUpdateScript [
   &ParticleState()
   &ScaleSpriteSize(Uniform Curve Sprite Scale=DataInterface() {propertyValues {Curve {
       defaultValue=1
       keys [
        {interpMode=RCIM_Cubic, tangentMode=RCTM_User, time=0, value=0}
        {arriveTangentWeight=0.094, interpMode=RCIM_Cubic, leaveTangentWeight=0.282, tangentMode=RCTM_User, time=0.1, value=0.9128}
        {arriveTangentWeight=0.3034, interpMode=RCIM_Cubic, leaveTangentWeight=0.1011, tangentMode=RCTM_User, time=0.9, value=0.7953}
        {tangentMode=RCTM_User, time=1, value=0}
       ]
      }}})
   &ScaleSpriteSize(
    Scale Sprite Size Mode=::NewEnumerator0
    Uniform Scale Factor=&Add_Float(
     A=$System.Rail Float
     B=&MakeCustomFloatFromBool(Bool=&Boolean_OrOperation(Bool1=$User.bIsFrontEnd, Bool2=$User.bIsFrontEndPreview), True Float=1, False Float=0)
    )
   ) {module=ScaleSpriteSize001}
   &ScaleColor(
    Scale Alpha=&Multiply_Float(
     A=&FloatFromCurve(
      FloatCurve=DataInterface() {propertyValues {Curve {
         defaultValue=0
         keys [
          {interpMode=RCIM_Cubic, tangentMode=RCTM_User, time=0, value=0}
          {interpMode=RCIM_Cubic, tangentMode=RCTM_User, time=0.5, value=1}
          {tangentMode=RCTM_User, time=1, value=0}
         ]
        }}}
      Scale Curve=&MaskFloat(Mask Source=1, Mask Parameter=$Particles.PartitionParticles.Partition, Mask=0, Masked Float=0, Passthrough Float=1)
     )
     B=&MergeHelper_FloatSwitch(
      0=$User.OuterCircle
      1=$User.MiddleCircle
      2=$User.MiddleCircle
      3=$User.InnerCircle
      4=$User.InnerCircle
      5=$User.InnerCircle
      Index=&SpawnGroup()
      Num=::NewEnumerator4
     )
    )
    Scale RGB=&MakeVectorFromLinearColorRGB(Color=&Multiply_LinearColorByFloat(
      Linear Color=&MergeHelper_ColorSwitch(
       0=&ColorFromCurve(
        ColorCurve=DataInterface() {propertyValues {
          AlphaCurve {keys [<at 0 1 RCIM_Constant>,<at 1 0 RCIM_Cubic>]}
          BlueCurve {keys [<at 0 0 RCIM_Constant>,<at 0.0628 0 RCIM_Constant>,<at 0.1402 0 RCIM_Constant>,<at 0.2131 0 RCIM_Cubic>,<at 0.2738 0 RCIM_Cubic>,<at 0.3967 0 RCIM_Cubic>,<at 0.5303 0 RCIM_Cubic>,<at 0.6016 0 RCIM_Constant>,<at 0.7033 0 RCIM_Cubic>,<at 0.8718 0 RCIM_Cubic>,<at 0.9902 0 RCIM_Cubic>]}
          GreenCurve {keys [
            <at 0 0.638 RCIM_Constant>
            <at 0.0628 0.0331 RCIM_Constant>
            <at 0.1402 0.6376 RCIM_Constant>
            {arriveTangent=-4.5259, interpMode=RCIM_Cubic, leaveTangent=-4.5259, time=0.2131, value=0.6376}
            {arriveTangent=-3.2915, interpMode=RCIM_Cubic, leaveTangent=-3.2915, time=0.2738, value=0.0331}
            {arriveTangent=2.3567, interpMode=RCIM_Cubic, leaveTangent=2.3567, time=0.3967, value=0.0331}
            {arriveTangent=2.9502, interpMode=RCIM_Cubic, leaveTangent=2.9502, time=0.5303, value=0.6376}
            <at 0.6016 0.6376 RCIM_Constant>
            {arriveTangent=-2.2375, interpMode=RCIM_Cubic, leaveTangent=-2.2375, time=0.7033, value=0.0331}
            {arriveTangent=2.1073, interpMode=RCIM_Cubic, leaveTangent=2.1073, time=0.8718, value=0.0331}
            <at 0.9902 0.6376 RCIM_Cubic>
           ]}
          RedCurve {keys [<at 0 1 RCIM_Constant>,<at 0.0628 1 RCIM_Constant>,<at 0.1402 1 RCIM_Constant>,<at 0.2131 1 RCIM_Cubic>,<at 0.2738 1 RCIM_Cubic>,<at 0.3967 1 RCIM_Cubic>,<at 0.5303 1 RCIM_Cubic>,<at 0.6016 1 RCIM_Constant>,<at 0.7033 1 RCIM_Cubic>,<at 0.8718 1 RCIM_Cubic>,<at 0.9902 1 RCIM_Cubic>]}
         }}
        CurveIndex=$System.NormalizedLoopAge
       )
       1=&ColorFromCurve(
        ColorCurve=DataInterface() {propertyValues {
          AlphaCurve {keys [<at 0 1 RCIM_Constant>,<at 1 0 RCIM_Cubic>]}
          BlueCurve {keys [<at 0.0492 0 RCIM_Constant>,<at 0.1114 0 RCIM_Constant>,<at 0.2101 0>,<at 0.2829 0>,<at 0.3694 0>,<at 0.4848 0>,<at 0.5743 0 RCIM_Cubic>,<at 0.6578 0 RCIM_Constant>,<at 0.7337 0>,<at 0.9462 0>,<at 0.9902 0>]}
          GreenCurve {keys [
            <at 0.0492 0.638 RCIM_Constant>
            <at 0.1114 0.0331 RCIM_Constant>
            <at 0.2101 0.6376>
            <at 0.2829 0.6376>
            <at 0.3694 0.0331>
            <at 0.4848 0.0331>
            {arriveTangent=3.4936, interpMode=RCIM_Cubic, leaveTangent=3.4936, time=0.5743, value=0.6376}
            <at 0.6578 0.6376 RCIM_Constant>
            <at 0.7337 0.0331>
            <at 0.9462 0.0331>
            <at 0.9902 0.6376>
           ]}
          RedCurve {keys [<at 0.0492 1 RCIM_Constant>,<at 0.1114 1 RCIM_Constant>,<at 0.2101 1>,<at 0.2829 1>,<at 0.3694 1>,<at 0.4848 1>,<at 0.5743 1 RCIM_Cubic>,<at 0.6578 1 RCIM_Constant>,<at 0.7337 1>,<at 0.9462 1>,<at 0.9902 1>]}
         }}
        CurveIndex=$System.NormalizedLoopAge
       )
       2=&ColorFromCurve(
        ColorCurve=DataInterface() {propertyValues {
          AlphaCurve {keys [<at 0 1 RCIM_Constant>,<at 1 0 RCIM_Cubic>]}
          BlueCurve {keys [<at 0.0932 0 RCIM_Constant>,<at 0.1569 0 RCIM_Constant>,<at 0.2662 0>,<at 0.2996 0>,<at 0.4286 0>,<at 0.4848 0>,<at 0.6108 0>,<at 0.6578 0 RCIM_Constant>,<at 0.805 0 RCIM_Constant>,<at 0.9462 0>,<at 0.9902 0>]}
          GreenCurve {keys [<at 0.0932 0.638 RCIM_Constant>,<at 0.1569 0.0331 RCIM_Constant>,<at 0.2662 0.6376>,<at 0.2996 0.6376>,<at 0.4286 0.0331>,<at 0.4848 0.0331>,<at 0.6108 0.6376>,<at 0.6578 0.6376 RCIM_Constant>,<at 0.805 0.0331 RCIM_Constant>,<at 0.9462 0.0331>,<at 0.9902 0.6376>]}
          RedCurve {keys [<at 0.0932 1 RCIM_Constant>,<at 0.1569 1 RCIM_Constant>,<at 0.2662 1>,<at 0.2996 1>,<at 0.4286 1>,<at 0.4848 1>,<at 0.6108 1>,<at 0.6578 1 RCIM_Constant>,<at 0.805 1 RCIM_Constant>,<at 0.9462 1>,<at 0.9902 1>]}
         }}
        CurveIndex=$System.NormalizedLoopAge
       )
       3=&ColorFromCurve(
        ColorCurve=DataInterface() {propertyValues {
          AlphaCurve {keys [<at 0 1 RCIM_Constant>,<at 1 0 RCIM_Cubic>]}
          BlueCurve {keys [<at 0.1357 0 RCIM_Constant>,<at 0.207 0 RCIM_Constant>,<at 0.3011 0>,<at 0.3527 0>,<at 0.4969 0>,<at 0.5485 0>,<at 0.6699 0>,<at 0.7853 0 RCIM_Constant>,<at 0.8551 0 RCIM_Constant>,<at 0.9583 0>,<at 0.9902 0>]}
          GreenCurve {keys [<at 0.1357 0.638 RCIM_Constant>,<at 0.207 0.0331 RCIM_Constant>,<at 0.3011 0.6376>,<at 0.3527 0.6376>,<at 0.4969 0.0331>,<at 0.5485 0.0331>,<at 0.6699 0.6376>,<at 0.7853 0.6376 RCIM_Constant>,<at 0.8551 0.0331 RCIM_Constant>,<at 0.9583 0.0331>,<at 0.9902 0.6376>]}
          RedCurve {keys [<at 0.1357 1 RCIM_Constant>,<at 0.207 1 RCIM_Constant>,<at 0.3011 1>,<at 0.3527 1>,<at 0.4969 1>,<at 0.5485 1>,<at 0.6699 1>,<at 0.7853 1 RCIM_Constant>,<at 0.8551 1 RCIM_Constant>,<at 0.9583 1>,<at 0.9902 1>]}
         }}
        CurveIndex=$System.NormalizedLoopAge
       )
       4=&ColorFromCurve(
        ColorCurve=DataInterface() {propertyValues {
          AlphaCurve {keys [<at 0 1 RCIM_Cubic>,<at 1 0 RCIM_Cubic>]}
          BlueCurve {keys [<at 0.1767 0 RCIM_Constant>,<at 0.2495 0 RCIM_Constant>,<at 0.3527 0 RCIM_Constant>,<at 0.4134 0>,<at 0.5485 0>,<at 0.6153 0>,<at 0.7352 0>,<at 0.7853 0 RCIM_Constant>,<at 0.9022 0 RCIM_Constant>,<at 0.9583 0>,<at 1 0>]}
          GreenCurve {keys [<at 0.1767 0.638 RCIM_Constant>,<at 0.2495 0.0331 RCIM_Constant>,<at 0.3527 0.6376 RCIM_Constant>,<at 0.4134 0.6376>,<at 0.5485 0.0331>,<at 0.6153 0.0331>,<at 0.7352 0.6376>,<at 0.7853 0.6376 RCIM_Constant>,<at 0.9022 0.0331 RCIM_Constant>,<at 0.9583 0.0331>,<at 1 0.6376>]}
          RedCurve {keys [<at 0.1767 1 RCIM_Constant>,<at 0.2495 1 RCIM_Constant>,<at 0.3527 1 RCIM_Constant>,<at 0.4134 1>,<at 0.5485 1>,<at 0.6153 1>,<at 0.7352 1>,<at 0.7853 1 RCIM_Constant>,<at 0.9022 1 RCIM_Constant>,<at 0.9583 1>,<at 1 1>]}
         }}
        CurveIndex=$System.NormalizedLoopAge
       )
       5=&ColorFromCurve(
        ColorCurve=DataInterface() {propertyValues {
          AlphaCurve {keys [<at 0 1 RCIM_Cubic>,<at 1 0 RCIM_Cubic>]}
          BlueCurve {keys [<at 0.2161 0 RCIM_Constant>,<at 0.3011 0 RCIM_Constant>,<at 0.3755 0 RCIM_Constant>,<at 0.4772 0>,<at 0.6199 0>,<at 0.6791 0>,<at 0.7853 0 RCIM_Constant>,<at 0.8384 0 RCIM_Constant>,<at 0.934 0 RCIM_Constant>,<at 0.972 0>,<at 1 0>]}
          GreenCurve {keys [<at 0.2161 0.638 RCIM_Constant>,<at 0.3011 0.0331 RCIM_Constant>,<at 0.3755 0.6376 RCIM_Constant>,<at 0.4772 0.6376>,<at 0.6199 0.0331>,<at 0.6791 0.0331>,<at 0.7853 0.6376 RCIM_Constant>,<at 0.8384 0.6376 RCIM_Constant>,<at 0.934 0.0331 RCIM_Constant>,<at 0.972 0.0331>,<at 1 0.6376>]}
          RedCurve {keys [<at 0.2161 1 RCIM_Constant>,<at 0.3011 1 RCIM_Constant>,<at 0.3755 1 RCIM_Constant>,<at 0.4772 1>,<at 0.6199 1>,<at 0.6791 1>,<at 0.7853 1 RCIM_Constant>,<at 0.8384 1 RCIM_Constant>,<at 0.934 1 RCIM_Constant>,<at 0.972 1>,<at 1 1>]}
         }}
        CurveIndex=$System.NormalizedLoopAge
       )
       Index=&SpawnGroup()
       Num=::NewEnumerator4
      )
      Float=10
     ))
    ScaleA=true
    ScaleRGB=true
   )
   &SolveForcesAndVelocity(Clamp Velocity=false, Limit Acceleration=false)
   &SubUVAnimation(
    Evaluation Type=::NewEnumerator0
    Output Alternate Renderer Binding=true
    "Sprite Renderer "=DataInterface() {propertyValues {SpriteRenderer=&self:Dot002_0.NiagaraSpriteRendererProperties_2}}
    SubUV Animation Mode:ENiagara_SubUVLookupModeV2=::NewEnumerator2
    UseEndFrame=false
    UseStartFrame=false
   )
   /SetVariables(
    Particles.GlowColor=&Multiply_LinearColorByFloat(
     Linear Color=$Particles.Color
     Float=&Multiply_Float(
      A=&FloatFromCurve(
       FloatCurve=DataInterface() {propertyValues {Curve {
          defaultValue=1
          keys [{interpMode=RCIM_Cubic, tangentMode=RCTM_User, time=0, value=1}, {interpMode=RCIM_Cubic, tangentMode=RCTM_User, time=1, value=0}]
         }}}
       Scale Curve=&RandomRangeFloat(Minimum=0.5, Maximum=0.7)
      )
      B=&MergeHelper_FloatSwitch(
       0=$User.OuterCircle
       1=$User.MiddleCircle
       2=$User.MiddleCircle
       3=$User.InnerCircle
       4=$User.InnerCircle
       5=$User.InnerCircle
       Index=&SpawnGroup()
       Num=::NewEnumerator4
      )
     )
    )
    Particles.GlowSize=&Multiply_Vector2D_ByFloat(Float=&RandomRangeFloat(Minimum=3, Maximum=1.7), Vector2D=$Particles.SpriteSize)
   ) {isSetParameters=true, module=SetVariables_#4}
   /SetVariables(
    Particles.RaysColor:LinearColor=&Multiply_LinearColorByFloat(
     Linear Color=$Particles.Color
     Float=&Multiply_Float(
      A=&FloatFromCurve(
       FloatCurve=DataInterface() {propertyValues {Curve {
          defaultValue=1
          keys [{interpMode=RCIM_Cubic, tangentMode=RCTM_User, time=0, value=1}, {interpMode=RCIM_Cubic, tangentMode=RCTM_User, time=1, value=0}]
         }}}
       Scale Curve=&RandomRangeFloat(Minimum=0.6, Maximum=1.2)
      )
      B=&MergeHelper_FloatSwitch(
       0=$User.OuterCircle
       1=$User.MiddleCircle
       2=$User.MiddleCircle
       3=$User.InnerCircle
       4=$User.InnerCircle
       5=$User.InnerCircle
       Index=&SpawnGroup()
       Num=::NewEnumerator4
      )
     )
    )
    Particles.RaysSize:Vector2f=&Multiply_Vector2D(A=&RandomRangeVector2D(Minimum=<1.8 2.5>, Maximum=<2.8 5>), B=$Particles.SpriteSize)
   ) {isSetParameters=true, module=SetVariables_#5}
  ]
  rendererClasses [&NiagaraSpriteRendererProperties]
  renderers [
   {rendererClass=&NiagaraSpriteRendererProperties}
   {rendererClass=&NiagaraSpriteRendererProperties, rendererIndex=1}
   {rendererClass=&NiagaraSpriteRendererProperties, rendererIndex=2}
  ]
 }
 Dot007 {
  bEnabled=false
  simTarget=GPUComputeSim
  emitterUpdateScript [
   &EmitterState(
    Life Cycle Mode=::NewEnumerator1
    Loop Delay=&RandomRangeFloat(Minimum=0.2, Maximum=0.5)
    Loop Duration=1
    UseLoopDelay=false
   )
   &UpdateValueStateChange(InputIntParam:NiagaraInt32=&MakeIntFromFloat(Float=$User.DiscAnim, Conversion Mode=::NewEnumerator1))
   &SpawnRate(SpawnRate=45, Use Spawn Probability=false)
   &SpawnRate(SpawnGroup=1, SpawnRate=40, Use Spawn Probability=false) {module=SpawnRate001}
   &SpawnRate(SpawnGroup=2, SpawnRate=30, Use Spawn Probability=false) {module=SpawnRate002}
   &SpawnRate(SpawnGroup=3, SpawnRate=20, Use Spawn Probability=false) {module=SpawnRate003}
   &SpawnRate(SpawnGroup=4, SpawnRate=12, Use Spawn Probability=false) {module=SpawnRate004}
   &SpawnRate(SpawnGroup=5, SpawnRate=4, Use Spawn Probability=false) {module=SpawnRate005}
  ]
  particleSpawnScript [
   &PartitionParticles(Number Of Partitions=5)
   &InitializeParticle(
    Lifetime=0.7
    Sprite Size Mode=::NewEnumerator3
    Uniform Sprite Size=&MergeHelper_FloatSwitch(0=7.5, 1=7.1, 2=6.7, 3=6.2, 4=5.8, 5=5.5, Index=&SpawnGroup(), Num=::NewEnumerator4)
    UsePositionOffset=false
    Write Material Random=true
   )
   &SocketLocation(
    Orientation Sampling=::NewEnumerator1
    Position Sampling=::NewEnumerator1
    SamplePosOffset=false
    Socket Reader=$System.Socket Reader
    Socket Sampling Mode=::NewEnumerator3
    VelocitySpeedLimit=false
   )
   &SpriteFacingAndAlignment(Facing Coordinate Space=::World, Sprite Facing=<0 0 1>, WriteAlignment=false, WriteFacing=true)
   &ShapeLocation(
    MaskSpawnGroup=false
    Offset=<0 0 5>
    Offset Coordinate Space=::World
    Offset Mode=::NewEnumerator0
    Override Seed=false
    Ring / Disc Distribution Mode:ENiagaraSphereDistributionMode=::NewEnumerator1
    Ring / Disc Mode:ENiagaraRingDiscMode=::NewEnumerator1
    Ring Radius=&MergeHelper_FloatSwitch(0=100, 1=80, 2=60, 3=40, 4=20, 5=6, Index=&SpawnGroup(), Num=::NewEnumerator4)
    Rotation Axis=<0 0 1>
    Rotation Coordinate Space=::World
    Rotation Mode=::NewEnumerator1
    Shape Primitive=::NewEnumerator4
    U Position=&Frac_Float(Float=&Multiply_Float(
      A=$Emitter.Age
      B=&Multiply_Float(
       A=$User.DiscAnim
       B=&FloatFromCurve(FloatCurve=DataInterface() {propertyValues {Curve {keys [<at 0 1>,<at 1 0>]}}})
      )
     ))
    UseOwnerScale=true
   ) {module=ShapeLocation002}
  ]
  particleUpdateScript [
   &ParticleState()
   &ScaleSpriteSize(Uniform Curve Sprite Scale=DataInterface() {propertyValues {Curve {
       defaultValue=1
       keys [
        {interpMode=RCIM_Cubic, tangentMode=RCTM_User, time=0, value=0}
        {arriveTangentWeight=0.094, interpMode=RCIM_Cubic, leaveTangentWeight=0.282, tangentMode=RCTM_User, time=0.1, value=0.9128}
        {arriveTangentWeight=0.3034, interpMode=RCIM_Cubic, leaveTangentWeight=0.1011, tangentMode=RCTM_User, time=0.9, value=0.7953}
        {tangentMode=RCTM_User, time=1, value=0}
       ]
      }}})
   &ScaleColor(
    Scale Alpha=&Multiply_Float(
     A=&FloatFromCurve(
      FloatCurve=DataInterface() {propertyValues {Curve {
         defaultValue=0
         keys [
          {interpMode=RCIM_Cubic, tangentMode=RCTM_User, time=0, value=0}
          {interpMode=RCIM_Cubic, tangentMode=RCTM_User, time=0.5, value=1}
          {tangentMode=RCTM_User, time=1, value=0}
         ]
        }}}
      Scale Curve=&MaskFloat(Mask Source=1, Mask Parameter=$Particles.PartitionParticles.Partition, Mask=0, Masked Float=0, Passthrough Float=1)
     )
     B=&MergeHelper_FloatSwitch(
      0=$User.OuterCircle
      1=$User.MiddleCircle
      2=$User.MiddleCircle
      3=$User.InnerCircle
      4=$User.InnerCircle
      5=$User.InnerCircle
      Index=&SpawnGroup()
      Num=::NewEnumerator4
     )
    )
    Scale RGB=&MakeVectorFromLinearColorRGB(Color=&Multiply_LinearColorByFloat(
      Linear Color=&MergeHelper_ColorSwitch(
       0=&ColorFromCurve(
        ColorCurve=DataInterface() {propertyValues {
          AlphaCurve {keys [<at 0 1 RCIM_Constant>,<at 1 0 RCIM_Cubic>]}
          BlueCurve {keys [<at 0 0 RCIM_Constant>,<at 0.0628 0 RCIM_Constant>,<at 0.1402 0 RCIM_Constant>,<at 0.2131 0 RCIM_Cubic>,<at 0.2738 0 RCIM_Cubic>,<at 0.3967 0 RCIM_Cubic>,<at 0.5303 0 RCIM_Cubic>,<at 0.6016 0 RCIM_Constant>,<at 0.7033 0 RCIM_Cubic>,<at 0.8718 0 RCIM_Cubic>,<at 0.9902 0 RCIM_Cubic>]}
          GreenCurve {keys [
            <at 0 0.638 RCIM_Constant>
            <at 0.0628 0.0331 RCIM_Constant>
            <at 0.1402 0.6376 RCIM_Constant>
            {arriveTangent=-4.5259, interpMode=RCIM_Cubic, leaveTangent=-4.5259, time=0.2131, value=0.6376}
            {arriveTangent=-3.2915, interpMode=RCIM_Cubic, leaveTangent=-3.2915, time=0.2738, value=0.0331}
            {arriveTangent=2.3567, interpMode=RCIM_Cubic, leaveTangent=2.3567, time=0.3967, value=0.0331}
            {arriveTangent=2.9502, interpMode=RCIM_Cubic, leaveTangent=2.9502, time=0.5303, value=0.6376}
            <at 0.6016 0.6376 RCIM_Constant>
            {arriveTangent=-2.2375, interpMode=RCIM_Cubic, leaveTangent=-2.2375, time=0.7033, value=0.0331}
            {arriveTangent=2.1073, interpMode=RCIM_Cubic, leaveTangent=2.1073, time=0.8718, value=0.0331}
            <at 0.9902 0.6376 RCIM_Cubic>
           ]}
          RedCurve {keys [<at 0 1 RCIM_Constant>,<at 0.0628 1 RCIM_Constant>,<at 0.1402 1 RCIM_Constant>,<at 0.2131 1 RCIM_Cubic>,<at 0.2738 1 RCIM_Cubic>,<at 0.3967 1 RCIM_Cubic>,<at 0.5303 1 RCIM_Cubic>,<at 0.6016 1 RCIM_Constant>,<at 0.7033 1 RCIM_Cubic>,<at 0.8718 1 RCIM_Cubic>,<at 0.9902 1 RCIM_Cubic>]}
         }}
        CurveIndex=$System.NormalizedLoopAge
       )
       1=&ColorFromCurve(
        ColorCurve=DataInterface() {propertyValues {
          AlphaCurve {keys [<at 0 1 RCIM_Constant>,<at 1 0 RCIM_Cubic>]}
          BlueCurve {keys [<at 0.0492 0 RCIM_Constant>,<at 0.1114 0 RCIM_Constant>,<at 0.2101 0>,<at 0.2829 0>,<at 0.3694 0>,<at 0.4848 0>,<at 0.5743 0 RCIM_Cubic>,<at 0.6578 0 RCIM_Constant>,<at 0.7337 0>,<at 0.9462 0>,<at 0.9902 0>]}
          GreenCurve {keys [
            <at 0.0492 0.638 RCIM_Constant>
            <at 0.1114 0.0331 RCIM_Constant>
            <at 0.2101 0.6376>
            <at 0.2829 0.6376>
            <at 0.3694 0.0331>
            <at 0.4848 0.0331>
            {arriveTangent=3.4936, interpMode=RCIM_Cubic, leaveTangent=3.4936, time=0.5743, value=0.6376}
            <at 0.6578 0.6376 RCIM_Constant>
            <at 0.7337 0.0331>
            <at 0.9462 0.0331>
            <at 0.9902 0.6376>
           ]}
          RedCurve {keys [<at 0.0492 1 RCIM_Constant>,<at 0.1114 1 RCIM_Constant>,<at 0.2101 1>,<at 0.2829 1>,<at 0.3694 1>,<at 0.4848 1>,<at 0.5743 1 RCIM_Cubic>,<at 0.6578 1 RCIM_Constant>,<at 0.7337 1>,<at 0.9462 1>,<at 0.9902 1>]}
         }}
        CurveIndex=$System.NormalizedLoopAge
       )
       2=&ColorFromCurve(
        ColorCurve=DataInterface() {propertyValues {
          AlphaCurve {keys [<at 0 1 RCIM_Constant>,<at 1 0 RCIM_Cubic>]}
          BlueCurve {keys [<at 0.0932 0 RCIM_Constant>,<at 0.1569 0 RCIM_Constant>,<at 0.2662 0>,<at 0.2996 0>,<at 0.4286 0>,<at 0.4848 0>,<at 0.6108 0>,<at 0.6578 0 RCIM_Constant>,<at 0.805 0 RCIM_Constant>,<at 0.9462 0>,<at 0.9902 0>]}
          GreenCurve {keys [<at 0.0932 0.638 RCIM_Constant>,<at 0.1569 0.0331 RCIM_Constant>,<at 0.2662 0.6376>,<at 0.2996 0.6376>,<at 0.4286 0.0331>,<at 0.4848 0.0331>,<at 0.6108 0.6376>,<at 0.6578 0.6376 RCIM_Constant>,<at 0.805 0.0331 RCIM_Constant>,<at 0.9462 0.0331>,<at 0.9902 0.6376>]}
          RedCurve {keys [<at 0.0932 1 RCIM_Constant>,<at 0.1569 1 RCIM_Constant>,<at 0.2662 1>,<at 0.2996 1>,<at 0.4286 1>,<at 0.4848 1>,<at 0.6108 1>,<at 0.6578 1 RCIM_Constant>,<at 0.805 1 RCIM_Constant>,<at 0.9462 1>,<at 0.9902 1>]}
         }}
        CurveIndex=$System.NormalizedLoopAge
       )
       3=&ColorFromCurve(
        ColorCurve=DataInterface() {propertyValues {
          AlphaCurve {keys [<at 0 1 RCIM_Constant>,<at 1 0 RCIM_Cubic>]}
          BlueCurve {keys [<at 0.1357 0 RCIM_Constant>,<at 0.207 0 RCIM_Constant>,<at 0.3011 0>,<at 0.3527 0>,<at 0.4969 0>,<at 0.5485 0>,<at 0.6699 0>,<at 0.7853 0 RCIM_Constant>,<at 0.8551 0 RCIM_Constant>,<at 0.9583 0>,<at 0.9902 0>]}
          GreenCurve {keys [<at 0.1357 0.638 RCIM_Constant>,<at 0.207 0.0331 RCIM_Constant>,<at 0.3011 0.6376>,<at 0.3527 0.6376>,<at 0.4969 0.0331>,<at 0.5485 0.0331>,<at 0.6699 0.6376>,<at 0.7853 0.6376 RCIM_Constant>,<at 0.8551 0.0331 RCIM_Constant>,<at 0.9583 0.0331>,<at 0.9902 0.6376>]}
          RedCurve {keys [<at 0.1357 1 RCIM_Constant>,<at 0.207 1 RCIM_Constant>,<at 0.3011 1>,<at 0.3527 1>,<at 0.4969 1>,<at 0.5485 1>,<at 0.6699 1>,<at 0.7853 1 RCIM_Constant>,<at 0.8551 1 RCIM_Constant>,<at 0.9583 1>,<at 0.9902 1>]}
         }}
        CurveIndex=$System.NormalizedLoopAge
       )
       4=&ColorFromCurve(
        ColorCurve=DataInterface() {propertyValues {
          AlphaCurve {keys [<at 0 1 RCIM_Cubic>,<at 1 0 RCIM_Cubic>]}
          BlueCurve {keys [<at 0.1767 0 RCIM_Constant>,<at 0.2495 0 RCIM_Constant>,<at 0.3527 0 RCIM_Constant>,<at 0.4134 0>,<at 0.5485 0>,<at 0.6153 0>,<at 0.7352 0>,<at 0.7853 0 RCIM_Constant>,<at 0.9022 0 RCIM_Constant>,<at 0.9583 0>,<at 1 0>]}
          GreenCurve {keys [<at 0.1767 0.638 RCIM_Constant>,<at 0.2495 0.0331 RCIM_Constant>,<at 0.3527 0.6376 RCIM_Constant>,<at 0.4134 0.6376>,<at 0.5485 0.0331>,<at 0.6153 0.0331>,<at 0.7352 0.6376>,<at 0.7853 0.6376 RCIM_Constant>,<at 0.9022 0.0331 RCIM_Constant>,<at 0.9583 0.0331>,<at 1 0.6376>]}
          RedCurve {keys [<at 0.1767 1 RCIM_Constant>,<at 0.2495 1 RCIM_Constant>,<at 0.3527 1 RCIM_Constant>,<at 0.4134 1>,<at 0.5485 1>,<at 0.6153 1>,<at 0.7352 1>,<at 0.7853 1 RCIM_Constant>,<at 0.9022 1 RCIM_Constant>,<at 0.9583 1>,<at 1 1>]}
         }}
        CurveIndex=$System.NormalizedLoopAge
       )
       5=&ColorFromCurve(
        ColorCurve=DataInterface() {propertyValues {
          AlphaCurve {keys [<at 0 1 RCIM_Cubic>,<at 1 0 RCIM_Cubic>]}
          BlueCurve {keys [<at 0.2161 0 RCIM_Constant>,<at 0.3011 0 RCIM_Constant>,<at 0.3755 0 RCIM_Constant>,<at 0.4772 0>,<at 0.6199 0>,<at 0.6791 0>,<at 0.7853 0 RCIM_Constant>,<at 0.8384 0 RCIM_Constant>,<at 0.934 0 RCIM_Constant>,<at 0.972 0>,<at 1 0>]}
          GreenCurve {keys [<at 0.2161 0.638 RCIM_Constant>,<at 0.3011 0.0331 RCIM_Constant>,<at 0.3755 0.6376 RCIM_Constant>,<at 0.4772 0.6376>,<at 0.6199 0.0331>,<at 0.6791 0.0331>,<at 0.7853 0.6376 RCIM_Constant>,<at 0.8384 0.6376 RCIM_Constant>,<at 0.934 0.0331 RCIM_Constant>,<at 0.972 0.0331>,<at 1 0.6376>]}
          RedCurve {keys [<at 0.2161 1 RCIM_Constant>,<at 0.3011 1 RCIM_Constant>,<at 0.3755 1 RCIM_Constant>,<at 0.4772 1>,<at 0.6199 1>,<at 0.6791 1>,<at 0.7853 1 RCIM_Constant>,<at 0.8384 1 RCIM_Constant>,<at 0.934 1 RCIM_Constant>,<at 0.972 1>,<at 1 1>]}
         }}
        CurveIndex=$System.NormalizedLoopAge
       )
       Index=&SpawnGroup()
       Num=::NewEnumerator4
      )
      Float=10
     ))
    ScaleA=true
    ScaleRGB=true
   )
   &SolveForcesAndVelocity(Clamp Velocity=false, Limit Acceleration=false)
   &SubUVAnimation(
    Evaluation Type=::NewEnumerator0
    Output Alternate Renderer Binding=true
    "Sprite Renderer "=DataInterface() {propertyValues {SpriteRenderer=&self:Dot007_1.NiagaraSpriteRendererProperties_2}}
    SubUV Animation Mode:ENiagara_SubUVLookupModeV2=::NewEnumerator2
    UseEndFrame=false
    UseStartFrame=false
   )
   /SetVariables(
    Particles.GlowColor=&Multiply_LinearColorByFloat(
     Linear Color=$Particles.Color
     Float=&Multiply_Float(
      A=&FloatFromCurve(
       FloatCurve=DataInterface() {propertyValues {Curve {
          defaultValue=1
          keys [{interpMode=RCIM_Cubic, tangentMode=RCTM_User, time=0, value=1}, {interpMode=RCIM_Cubic, tangentMode=RCTM_User, time=1, value=0}]
         }}}
       Scale Curve=&RandomRangeFloat(Minimum=0.5, Maximum=0.7)
      )
      B=&MergeHelper_FloatSwitch(
       0=$User.OuterCircle
       1=$User.MiddleCircle
       2=$User.MiddleCircle
       3=$User.InnerCircle
       4=$User.InnerCircle
       5=$User.InnerCircle
       Index=&SpawnGroup()
       Num=::NewEnumerator4
      )
     )
    )
    Particles.GlowSize=&Multiply_Vector2D_ByFloat(Float=&RandomRangeFloat(Minimum=3, Maximum=1.7), Vector2D=$Particles.SpriteSize)
   ) {isSetParameters=true, module=SetVariables_#4}
   /SetVariables(
    Particles.RaysColor:LinearColor=&Multiply_LinearColorByFloat(
     Linear Color=$Particles.Color
     Float=&Multiply_Float(
      A=&FloatFromCurve(
       FloatCurve=DataInterface() {propertyValues {Curve {
          defaultValue=1
          keys [{interpMode=RCIM_Cubic, tangentMode=RCTM_User, time=0, value=1}, {interpMode=RCIM_Cubic, tangentMode=RCTM_User, time=1, value=0}]
         }}}
       Scale Curve=&RandomRangeFloat(Minimum=0.6, Maximum=1.2)
      )
      B=&MergeHelper_FloatSwitch(
       0=$User.OuterCircle
       1=$User.MiddleCircle
       2=$User.MiddleCircle
       3=$User.InnerCircle
       4=$User.InnerCircle
       5=$User.InnerCircle
       Index=&SpawnGroup()
       Num=::NewEnumerator4
      )
     )
    )
    Particles.RaysSize:Vector2f=&Multiply_Vector2D(A=&RandomRangeVector2D(Minimum=<1.8 2.5>, Maximum=<2.8 5>), B=$Particles.SpriteSize)
   ) {isSetParameters=true, module=SetVariables_#5}
  ]
  rendererClasses [&NiagaraSpriteRendererProperties]
  renderers [
   {rendererClass=&NiagaraSpriteRendererProperties}
   {rendererClass=&NiagaraSpriteRendererProperties, rendererIndex=1}
   {rendererClass=&NiagaraSpriteRendererProperties, rendererIndex=2}
  ]
 }
 Light001 {
  emitterUpdateScript [
   &EmitterState(Life Cycle Mode=::NewEnumerator1, Loop Behavior=::NewEnumerator1, Loop Duration Mode=::NewEnumerator1, UseLoopDelay=false)
   &SpawnBurst_Instantaneous(Use Spawn Probability=false, UseLoopCountLimit=false)
  ]
  particleSpawnScript [
   &InitializeParticle(Sprite Size Mode=::NewEnumerator3, Uniform Sprite Size=30, UsePositionOffset=false, Write Material Random=true)
   /SetVariables(Particles.LightExponent=2, Particles.LightRadius=80) {isSetParameters=true}
  ]
  particleUpdateScript [
   &ParticleState(Kill Particles When Lifetime Has Elapsed=false, Let Infinitely Lived Particles Die When Emitter Deactivates=true)
   &ScaleColor(
    Scale RGB=&MakeVectorFromLinearColorRGB(Color=&Multiply_LinearColorByFloat(
      Linear Color=&ColorFromCurve(
       ColorCurve=DataInterface() {propertyValues {
         AlphaCurve {keys [<at 0 1 RCIM_Cubic>,<at 1 0 RCIM_Cubic>]}
         BlueCurve {keys [<at 0.1767 0 RCIM_Cubic>,<at 0.2495 0 RCIM_Cubic>,<at 0.3527 0 RCIM_Cubic>,<at 0.4134 0 RCIM_Cubic>,<at 0.5485 0 RCIM_Cubic>,<at 0.6153 0 RCIM_Cubic>,<at 0.7352 0 RCIM_Cubic>,<at 0.7853 0 RCIM_Cubic>,<at 0.9022 0 RCIM_Cubic>,<at 0.9583 0 RCIM_Cubic>,<at 1 0 RCIM_Cubic>]}
         GreenCurve {keys [
           <at 0.1767 0.638 RCIM_Cubic>
           {arriveTangent=-0.002, interpMode=RCIM_Cubic, leaveTangent=-0.002, time=0.2495, value=0.0331}
           {arriveTangent=3.6877, interpMode=RCIM_Cubic, leaveTangent=3.6877, time=0.3527, value=0.6376}
           {arriveTangent=-3.0874, interpMode=RCIM_Cubic, leaveTangent=-3.0874, time=0.4134, value=0.6376}
           {arriveTangent=-2.9946, interpMode=RCIM_Cubic, leaveTangent=-2.9946, time=0.5485, value=0.0331}
           {arriveTangent=3.238, interpMode=RCIM_Cubic, leaveTangent=3.238, time=0.6153, value=0.0331}
           {arriveTangent=3.556, interpMode=RCIM_Cubic, leaveTangent=3.556, time=0.7352, value=0.6376}
           {arriveTangent=-3.6207, interpMode=RCIM_Cubic, leaveTangent=-3.6207, time=0.7853, value=0.6376}
           {arriveTangent=-3.4936, interpMode=RCIM_Cubic, leaveTangent=-3.4936, time=0.9022, value=0.0331}
           {arriveTangent=6.1786, interpMode=RCIM_Cubic, leaveTangent=6.1786, time=0.9583, value=0.0331}
           <at 1 0.6376 RCIM_Cubic>
          ]}
         RedCurve {keys [<at 0.1767 1 RCIM_Cubic>,<at 0.2495 1 RCIM_Cubic>,<at 0.3527 1 RCIM_Cubic>,<at 0.4134 1 RCIM_Cubic>,<at 0.5485 1 RCIM_Cubic>,<at 0.6153 1 RCIM_Cubic>,<at 0.7352 1 RCIM_Cubic>,<at 0.7853 1 RCIM_Cubic>,<at 0.9022 1 RCIM_Cubic>,<at 0.9583 1 RCIM_Cubic>,<at 1 1 RCIM_Cubic>]}
        }}
       CurveIndex=$System.NormalizedLoopAge
      )
      Float=3
     ))
    ScaleA=true
    ScaleRGB=true
   )
  ]
  rendererClasses [&NiagaraLightRendererProperties]
  renderers [{rendererClass=&NiagaraLightRendererProperties, rendererIndex=0}]
 }
}
summary {
 systemName=NS_Emote_Noise_ClueLurch_Dot
 emitters [
  {emitterName=Light001, rendererClasses [&NiagaraLightRendererProperties]}
  {bEnabled=false, emitterName=Dot007, simTarget=GPUComputeSim, rendererClasses [&NiagaraSpriteRendererProperties]}
  {bEnabled=false, emitterName=Dot, rendererClasses [&NiagaraSpriteRendererProperties]}
  {emitterName=Dot001, rendererClasses [&NiagaraSpriteRendererProperties]}
  {emitterName=Dot002, rendererClasses [&NiagaraSpriteRendererProperties]}
 ]
 userVariables [
  User.bIsFrontEnd {type=NiagaraBool}
  User.bIsFrontEndPreview {type=NiagaraBool}
  User.DiscAnim {type=NiagaraFloat, defaultValue=1, description=""}
  User.InnerCircle {type=NiagaraFloat, defaultValue=1, description=""}
  User.MiddleCircle {type=NiagaraFloat, defaultValue=1, description=""}
  User.OuterCircle {type=NiagaraFloat, defaultValue=1, description=""}
  User.OwnerLocalVelocity {type=Vector3f, description="", defaultValue {struct=Vector3f, value=<0 0 0>}}
 ]
}
